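# Flat metal material for the medium male player Dummy body, usable on skeletal meshes, Nanite and mesh deformers.
# It uses no textures, only constants for a fully metallic dark grey with moderate roughness, packed into material attributes.
# Those attributes chain through MF_CreativeMoveTool and then MF_TechArtOverrides, layering shared functions over the base look.
# It is Default Lit and assigns the Metal physical material.
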Material /Game/Characters/Player/Male/Medium/Bodies/M_MED_Dummy/Materials/M_MED_Dummy_Metal
graph {
 properties {
  bUseMaterialAttributes=true
  bUsedWithMeshDeformer=true
  bUsedWithNanite=true
  bUsedWithSkeletalMesh=true
  physMaterial=/Game/FortressPhysicalMaterials/Metal
  pixelDepthOffsetMode=PDOM_Legacy
  usedShadingModels=Default Lit
 }
 nodes [
  n1 = Constant() {r=0.3}
  n2 = Constant() {r=1}
  MaterialFunctionCall(In=MaterialFunctionCall(MA In=MakeMaterialAttributes!(
     BaseColor=@n1
     Metallic=@n2
     Specular
     Roughness=@n1
     Anisotropy
     EmissiveColor
     Opacity
     OpacityMask
     Normal
     Tangent
     WorldPositionOffset
     SubsurfaceColor
     ClearCoat
     ClearCoatRoughness
     AmbientOcclusion
     Refraction
     CustomizedUVs_0
     CustomizedUVs_1
     CustomizedUVs_2
     CustomizedUVs_3
     CustomizedUVs_4
     CustomizedUVs_5
     CustomizedUVs_6
     CustomizedUVs_7
     PixelDepthOffset
     ShadingModel
     Displacement
    )) {materialFunction=&MF_CreativeMoveTool}.MA Out) {materialFunction=&MF_TechArtOverrides}
 ]
}
