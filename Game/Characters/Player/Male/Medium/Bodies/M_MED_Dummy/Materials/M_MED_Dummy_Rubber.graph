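# Opaque rubber material for the M_MED_Dummy male medium player body, usable on skeletal meshes and with mesh deformers.
# A constant near-black base color and high constant roughness are written through SetMaterialAttributes, with no textures.
# The attributes pass through MF_TechArtOverrides as the final step of the Default Lit graph.
# Despite the rubber look, the assigned physical material is Metal.
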Material /Game/Characters/Player/Male/Medium/Bodies/M_MED_Dummy/Materials/M_MED_Dummy_Rubber
graph {
 properties {
  bUseMaterialAttributes=true
  bUsedWithMeshDeformer=true
  bUsedWithSkeletalMesh=true
  physMaterial=/Game/FortressPhysicalMaterials/Metal
  pixelDepthOffsetMode=PDOM_Legacy
  usedShadingModels=Default Lit
 }
 nodes [MaterialFunctionCall(In=SetMaterialAttributes!(MaterialAttributes, Base Color=Constant3Vector() {constant=<rgba 0.026 0.026 0.026 0>}, Roughness=Constant() {r=0.8}) {attributeSetTypes [#1, #2]}) {materialFunction=&MF_TechArtOverrides}]
}
